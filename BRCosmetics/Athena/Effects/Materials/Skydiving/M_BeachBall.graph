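# Default Lit beach ball material in the skydiving effects folder, enabled for mesh particles in Cascade and Niagara.
# It exposes Diffuse, Normal and Specular texture parameters that default to the SportingGoods beach ball prop textures.
# The Specular mask's red channel is squared, scaled down and clamped to a narrower range, alongside two standalone scalar constants.
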Material /BRCosmetics/Athena/Effects/Materials/Skydiving/M_BeachBall
graph {
 properties {bUsedWithMeshParticles=true, bUsedWithNiagaraMeshParticles=true, pixelDepthOffsetMode=PDOM_Legacy, usedShadingModels=Default Lit}
 nodes [
  n1 = TextureSampleParameter2D() {parameterName=Specular, samplerType=SAMPLERTYPE_Masks, texture=/Game/Environments/Props/SportingGoods/Textures/T_SportingGoods_BeachBall_S}
  Clamp(None=Multiply(A=Multiply(A=@n1.R, B=@n1.R)) {constB=0.8}) {maxDefault=0.9, minDefault=0.1}
  Constant() {r=0.4}
  Constant()
  TextureSampleParameter2D() {parameterName=Diffuse, texture=/Game/Environments/Props/SportingGoods/Textures/T_SportingGoods_BeachBall_D}
  TextureSampleParameter2D() {parameterName=Normal, samplerType=SAMPLERTYPE_Normal, texture=/Game/Environments/Props/SportingGoods/Textures/T_SportingGoods_BeachBall_N}
 ]
}
