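# Translucent unlit particle material for skydiving effects, rendering soft, tinted smoke with index-of-refraction distortion on sprite and Niagara particles.
# Particle color times the smoke texture and a 2D sphere gradient feeds an emissive adjustment function.
# Particle alpha times the smoke sets opacity, scaled by a parameter, and also blends refraction toward a strength parameter.
# A quality switch outputs neutral refraction on low settings, and a pointy normal map is also sampled.
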Material /BRCosmetics/Athena/Effects/Materials/Skydiving/M_BluRefrac
graph {
 properties {
  bUsedWithNiagaraSprites=true
  bUsedWithParticleSprites=true
  blendMode=BLEND_Translucent
  pixelDepthOffsetMode=PDOM_Legacy
  refractionMethod=RM_IndexOfRefraction
  shadingModel=MSM_Unlit
  usedShadingModels=Unlit
 }
 nodes [
  n1 = ParticleColor()
  n2 = TextureSample() {texture=/Game/Effects/Fort_Effects/Textures/T_Smoke_Soft}
  n3 = MaterialFunctionCall!(UVs, CenterPosition, Radius) {materialFunction=/Engine/Functions/Engine_MaterialFunctions01/Gradient/SphereGradient-2D}
  MaterialFunctionCall(Emissive=Multiply(A=@n1.RGB, B=Multiply(A=@n2.G, B=@n3.Result 0-1))) {materialFunction=&MF_FX_EmissiveAdjustments}
  Multiply(A=Multiply(A=@n1.A, B=@n2.G), B=ScalarParameter() {defaultValue=0.2714, parameterName=OverallColorOpacity})
  n4 = LinearInterpolate(B=ScalarParameter() {defaultValue=0.95, parameterName=RefracStr}, Alpha=Multiply(A=@n2.G, B=ParticleColor().A)) {constA=1, constB=0.9}
  QualitySwitch(Default=@n4, Low=Constant() {r=1}, High=@n4, Medium=@n4)
  TextureSample() {samplerType=SAMPLERTYPE_Normal, texture=/Game/Effects/Fort_Effects/Textures/PointyNormal}
 ]
}
