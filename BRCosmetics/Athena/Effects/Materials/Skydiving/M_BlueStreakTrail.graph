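# Unlit translucent material for a blue skydiving streak trail, usable on beam trails and Niagara ribbons.
# Emissive combines a panning trail mask tinted by glow color with rotating zoom-noise light shafts, then applies power, a glow-pass color and optional particle-color multiply.
# Opacity multiplies a separately panned sample of that mask by particle alpha, camera depth fade and a scalar.
# Dynamic parameters give color and alpha independent random U offsets.
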Material /BRCosmetics/Athena/Effects/Materials/Skydiving/M_BlueStreakTrail
graph {
 properties {
  bUsedWithBeamTrails=true
  bUsedWithNiagaraRibbons=true
  blendMode=BLEND_Translucent
  pixelDepthOffsetMode=PDOM_Legacy
  shadingModel=MSM_Unlit
  usedShadingModels=Unlit
 }
 nodes [
  n1 = ParticleColor()
  n2 = TextureCoordinate() {coordinateIndex=1}
  n3 = AppendVector(A=DynamicParameter() {paramNames [RandomStartU_Color, RandomStartU_Alpha, Param3, Param4]}.RandomStartU_Color, B=Constant())
  n4 = VectorParameter() {parameterName=TexScale, defaultValue=<rgba 0.25 1 0 1>}
  n5 = Panner(
   Coordinate=Multiply(A=Add(A=@n2, B=@n3), B=ComponentMask(None=@n4.RGB) {g=true, r=true})
   Speed=AppendVector(A=ScalarParameter() {defaultValue=-1, parameterName=PanSpeedU}, B=ScalarParameter() {parameterName=PanSpeedV})
  ) {speedX=0.25}
  n6 = VectorParameter() {parameterName=2D Glow color, defaultValue=<rgba 0.58 0.9576 2 2>}
  n7 = Multiply(A=Clamp(None=TextureSample(UVs=@n5) {texture=/Game/Effects/Fort_Effects/Textures/Smoke/T_Lobber_Trail_Mask}.G), B=@n6.RGB)
  n8 = TextureCoordinate()
  n9 = Time()
  n10 = TextureSample(UVs=Rotator(Coordinate=@n8, Time=@n9) {speed=-0.1}) {texture=/Game/Effects/Fort_Effects/BluGlow/ZoomNoise}
  n11 = Multiply(A=TextureSample(UVs=Rotator(Coordinate=@n8, Time=@n9)) {texture=&ZoomNoise}.R, B=Add(A=@n10.G, B=@n10.G))
  n12 = Multiply(
   A=LinearInterpolate(
    A=VectorParameter() {parameterName=Light shaft Color 1, defaultValue=<rgba 0.4121 0.4795 0.87 1>}.RGB
    B=VectorParameter() {parameterName=Light shaft Color 2, defaultValue=<rgba 0.8597 0.905 0.8911 1>}.RGB
    Alpha=@n11
   )
   B=Multiply(A=@n11, B=ScalarParameter() {defaultValue=1, desc=For Blueprints, parameterName=LightShaftMultipler}) {constB=2.233}
  )
  n13 = Add(
   A=Power(
    Base=MaterialFunctionCall!(Base=@n7, Add=@n12) {materialFunction=/Engine/Functions/Engine_MaterialFunctions02/Math/MultiplyAdd}.Result
    Exp=@n6.A
   )
   B=VectorParameter() {parameterName=Glow Pass, defaultValue=<rgba 0 0.0314 0.27 0.1>}.RGB
  )
  MaterialFunctionCall(Emissive=LinearInterpolate(A=@n13, B=Multiply(A=@n13, B=@n1.RGB), Alpha=ScalarParameter() {parameterName=Use_ParticleColor_Multiply})) {materialFunction=&MF_FX_EmissiveAdjustments}
  n14 = Add(
   A=TextureCoordinate() {coordinateIndex=1}
   B=AppendVector(A=DynamicParameter() {paramNames [RandomStartU_Color, RandomStartU_Alpha, Param3, Param4]}.RandomStartU_Alpha, B=Constant())
  )
  n15 = AppendVector(A=ScalarParameter() {defaultValue=-1, parameterName=PanSpeedU}, B=ScalarParameter() {parameterName=PanSpeedV})
  n16 = Clamp(None=TextureSample(UVs=Panner(Coordinate=@n14, Speed=@n15)) {texture=&T_Lobber_Trail_Mask}.G)
  n17 = ScalarParameter() {defaultValue=75, parameterName=CamFade}
  Multiply(
   A=Multiply(
    A=Multiply(A=@n1.A, B=@n16)
    B=MaterialFunctionCall(Fade Length=@n17) {materialFunction=&CameraDepthFade}.Result
   )
   B=ScalarParameter() {defaultValue=0.5, parameterName=OpacityMult}
  )
 ]
}
