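# Translucent unlit, two-sided skydiving contrail material for beam and Niagara ribbon trails, with a celestial starfield and nebula look.
# Emissive blends dark tints and nebula through panning inky smoke, adds stars, and optionally darkens edges, using camera-vector triplanar projection or screen-aligned UVs on mobile.
# Opacity lighten-blends two panning smoke samples, offset by object-relative sphere mask and dynamic parameter, then applies UV masks, pulse, particle alpha and camera fades.
# Mobile shading paths skip depth fades.
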Material /BRCosmetics/Athena/Effects/Materials/Skydiving/M_Contrail_Celestial
graph {
 properties {
  bUsedWithBeamTrails=true
  bUsedWithNiagaraRibbons=true
  blendMode=BLEND_Translucent
  pixelDepthOffsetMode=PDOM_Legacy
  shadingModel=MSM_Unlit
  twoSided=true
  usedShadingModels=Unlit
 }
 nodes [
  n1 = Sine(None=Multiply(A=Constant() {r=0.22}, B=Time()))
  n2 = DynamicParameter() {paramNames [AlphaMask_U,Param2,Param3,Param4]}
  n3 = SphereMask(
   A=WorldPosition().XYZ
   B=Add(A=ObjectPositionWS(), B=VectorParameter() {parameterName=InitialOffset, defaultValue=<rgba 0 0 -1000 1>}.RGB)
   Radius=Constant() {r=2550}
   Hardness=Constant()
  )
  n4 = Add(A=Multiply(A=@n3, B=@n3), B=Constant() {r=-200}) {constB=-20}
  n5 = Add(A=@n4, B=@n2.AlphaMask_U)
  n6 = Multiply(A=ComponentMask(None=TextureCoordinate() {coordinateIndex=1}) {g=true})
  n7 = Multiply(A=ScalarParameter() {defaultValue=-0.5, parameterName=TimeMutiply}, B=Time())
  n8 = Panner(Coordinate=AppendVector(A=Multiply(A=@n5) {constB=7}, B=@n6), Time=@n7) {speedX=-2}
  n9 = Add(A=Add(A=@n2.AlphaMask_U, B=@n4)) {constB=0.6}
  n10 = Panner(Coordinate=AppendVector(A=Multiply(A=@n9) {constB=9}, B=@n6), Time=@n7) {speedX=-2}
  n11 = Multiply(A=MaterialFunctionCall!(
    Blend=TextureSample(UVs=@n8) {samplerType=SAMPLERTYPE_Grayscale, texture=&T_Smoke_Trail_Soft}.G
    Base=TextureSample(UVs=@n10) {samplerType=SAMPLERTYPE_Grayscale, texture=&T_Smoke_Trail_Soft}.G
   ) {materialFunction=/Engine/Functions/Engine_MaterialFunctions03/Blends/Blend_Lighten}.Result) {constB=1.25}
  n12 = Constant() {r=100}
  n13 = Add(
   A=ScalarParameter() {parameterName=CameraMask}
   B=MaterialFunctionCall(Fade Length=@n12, Fade Offset=@n12) {materialFunction=&CameraDepthFade}.Result
  )
  n14 = TextureCoordinate() {coordinateIndex=1}
  n15 = TextureCoordinate()
  n16 = SphereMask(
   A=Power(Base=MaterialFunctionCall(Float2=@n15) {materialFunction=&BreakOutFloat2Components}.R) {constExponent=0.25}
   B=Constant() {r=0.5}
  ) {attenuationRadius=0.5, hardnessPercent=0}
  n17 = Multiply(
   A=SphereMask(
    A=MaterialFunctionCall(Float2=@n14) {materialFunction=&BreakOutFloat2Components}.G
    B=Constant() {r=0.5}
   ) {attenuationRadius=0.5, hardnessPercent=0}
   B=Multiply(A=@n16, B=@n16)
  )
  n18 = ShadingPathSwitch(Default=Multiply(A=Clamp(None=@n13), B=@n17), Mobile=@n17) {desc=Migrated from FeatureLevelSwitch}
  n19 = Multiply(
   A=Clamp(None=Abs(None=@n1)) {minDefault=0.6}
   B=MaterialFunctionCall(Alpha=Multiply(A=@n11, B=@n18), Min=Constant(), Max=Constant() {r=0.75}) {materialFunction=&SmoothStep}.Result
  )
  n20 = ShadingPathSwitch(Default=DepthFade(Opacity=@n19), Mobile=@n19) {desc=Migrated from FeatureLevelSwitch}
  n21 = MaterialFunctionCall(Distance=ShadingPathSwitch(
    Default=ScalarParameter() {defaultValue=25, parameterName=NearFadeDistance}
    Mobile=ScalarParameter() {defaultValue=150, parameterName=NearFadeDistance_Mobile}
   ) {desc=Migrated from FeatureLevelSwitch}) {materialFunction=&Near_Camera_Fade}
  Clamp(None=Multiply(A=ParticleColor().A, B=Multiply(A=@n20, B=@n21.Result)))
  n22 = TextureObject() {samplerType=SAMPLERTYPE_Masks, texture=/Game/Characters/Player/Male/Medium/Bodies/M_Med_Soldier_04/Skins/Galaxy/Textures/Celestial_T_Stars}
  n23 = Time()
  n24 = ComponentMask(None=MaterialFunctionCall!(CameraVector, Texture=@n22, Tiling=Constant3Vector() {constant=<rgba 3 3 3 0>}, Offset=Multiply(A=@n23) {constB=0.02}, AxisFadeContrast) {materialFunction=/Engine/Functions/Engine_MaterialFunctions01/Texturing/TriplanarCameraVector}.Result) {r=true}
  n25 = TextureSample(UVs=Panner(Coordinate=TextureCoordinate() {uTiling=0.7, vTiling=0.8}) {speedX=0.2}) {texture=&T_Inky_Smoke_Tile}
  n26 = LinearInterpolate(A=Constant3Vector() {constant=<rgba 0.0015 0.0028 0.0104 1>}, B=Constant3Vector() {constant=<rgba 0.2135 0.0266 0.1457 1>}, Alpha=@n25.R)
  n27 = TextureObject() {texture=/Game/Characters/Player/Male/Medium/Bodies/M_Med_Soldier_04/Skins/Galaxy/Textures/T_FN_Nebula}
  n28 = Multiply(A=MaterialFunctionCall(Texture=@n27, Tiling=Constant3Vector() {constant=<rgba 1.5 1.5 1.5 0>}) {materialFunction=&TriplanarCameraVector}.Result) {constB=3}
  n29 = Constant() {r=128}
  n30 = TextureSample(
   UVs=Panner(Coordinate=MaterialFunctionCall!(TextureResolution=@n29) {materialFunction=/Engine/Functions/Engine_MaterialFunctions02/Texturing/ScreenAlignedPixelToPixelUVs}.Result) {speedX=0.1}
   Tex=@n22
  )
  n31 = Constant() {r=256}
  n32 = TextureSample(
   UVs=Panner(Coordinate=MaterialFunctionCall(TextureResolution=@n31) {materialFunction=&ScreenAlignedPixelToPixelUVs}.Result) {speedX=0.1}
   Tex=@n27
  )
  n33 = ShadingPathSwitch(
   Default=Add(A=Multiply(A=@n24) {constB=0.85}, B=LinearInterpolate(A=@n26, B=@n28, Alpha=@n25.R))
   Mobile=Add(A=ComponentMask(None=@n30.R) {r=true}, B=LinearInterpolate(A=@n26, B=@n32.RGB, Alpha=@n25.R))
  ) {desc=Migrated from FeatureLevelSwitch}
  n34 = ComponentMask(None=TextureCoordinate()) {r=true}
  n35 = Constant() {r=0.5}
  n36 = LinearInterpolate(A=Constant3Vector(), B=@n33, Alpha=Saturate(None=SphereMask(A=@n34, B=@n35) {attenuationRadius=0.55, hardnessPercent=0}))
  MaterialFunctionCall(Emissive=Multiply(A=StaticSwitchParameter(True=@n33, False=@n36) {parameterName=edge power}, B=Constant3Vector() {constant=<rgba 0.8 1 1.5 1>}) {constB=1.4}) {materialFunction=&MF_FX_EmissiveAdjustments}
  Multiply(A=@n17, B=SphereMask(A=TextureCoordinate(), B=Constant2Vector() {g=0.5, r=1.025}) {attenuationRadius=1.05, hardnessPercent=90})
  TextureSample(UVs=Multiply(A=TextureCoordinate() {coordinateIndex=1}, B=Constant2Vector() {g=1, r=5})) {samplerType=SAMPLERTYPE_Grayscale, texture=&T_Smoke_Trail_Soft}
 ]
}
