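# Lit Niagara mesh-particle material that gives a skydiving contrail a disco ball look, with mirror tiles and shifting colored glints.
# Panning DiscoTiles normal and mask textures supply facet detail, with scroll speed branched on a dynamic parameter.
# The tile mask plus time and another dynamic parameter hue-shifts red, gated by the inverted mask into boosted emissive.
# A UV-derived sphere normal dotted with a fixed direction adds fake shading to the disco ball diffuse texture.
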Material /BRCosmetics/Athena/Effects/Materials/Skydiving/M_Contrail_DiscoBall
graph {
 properties {bUsedWithNiagaraMeshParticles=true, pixelDepthOffsetMode=PDOM_Legacy, usedShadingModels=Default Lit}
 nodes [
  n1 = TextureCoordinate()
  n2 = Constant() {r=0.5}
  n3 = Constant()
  n4 = Multiply(A=Multiply(A=Subtract(A=@n1) {constB=0.5}, B=SphereMask(A=@n1, B=@n2, Radius=@n2, Hardness=@n3))) {constB=3.5}
  n5 = Constant() {r=1}
  n6 = TransformPosition(None=Normalize(VectorInput=AppendVector(A=@n4, B=@n5))) {transformType=TRANSFORMPOSSOURCE_World}
  n7 = Constant() {r=1}
  n8 = Constant() {r=-1}
  n9 = Normalize(VectorInput=AppendVector(A=AppendVector(A=@n7, B=@n8), B=Constant() {r=1}))
  Add(
   A=Multiply(A=DotProduct(A=@n6, B=@n9)) {constB=0.1}
   B=Multiply(A=TextureSample() {texture=/Game/Environments/Props/Retail/Textures/T_DiscoBall_D}.RGB) {constB=1.58}
  )
  n10 = DynamicParameter()
  n11 = If(A=@n10.Param4, "A > B"=@n10.Param4, "A < B"=Subtract(A=@n10.Param4) {constB=0.8}) {constB=0.5}
  n12 = Panner(Coordinate=TextureCoordinate() {uTiling=2, vTiling=1.5}, Speed=AppendVector(A=@n11, B=Constant())) {speedX=0.7}
  n13 = TextureSample(UVs=@n12) {samplerType=SAMPLERTYPE_Normal, texture=/Game/Global/Textures/Default/Tileable_Patterns/DiscoTiles_N}
  Add(
   A=@n13.RGB
   B=TextureSample(UVs=TextureCoordinate() {uTiling=2, vTiling=2}) {samplerType=SAMPLERTYPE_Normal, texture=/Game/Weapons/WeaponSkins/Wraps/Textures/DiscoBall_N}.RGB
  )
  Constant() {r=1}
  Constant()
  Constant()
  n14 = TextureCoordinate() {uTiling=2, vTiling=1.5}
  n15 = Panner(Coordinate=@n14, Speed=AppendVector(A=@n11, B=Constant())) {speedX=0.7}
  n16 = TextureSample(UVs=@n15) {samplerType=SAMPLERTYPE_LinearColor, texture=/Game/Global/Textures/Default/Tileable_Patterns/DiscoTiles_M}
  LinearInterpolate(
   B=LinearInterpolate(Alpha=@n16.R) {constA=0.28}
   Alpha=TextureSample(UVs=@n12) {samplerType=SAMPLERTYPE_LinearColor, texture=&DiscoTiles_M}.G
  ) {constA=-1, constB=0.5}
  n17 = Add(A=DynamicParameter().Param1, B=Time())
  n18 = Multiply(A=MaterialFunctionCall(Hue Shift Percentage=Add(A=@n16.R, B=@n17), Texture=Constant3Vector() {constant=<rgba 1 0 0 0>}) {materialFunction=&HueShift}.Result) {constB=0.5}
  n19 = TextureSample(UVs=@n15) {samplerType=SAMPLERTYPE_LinearColor, texture=&DiscoTiles_M}
  n20 = Time()
  n21 = LinearInterpolate(A=@n19.R, B=@n19.R, Alpha=Sine(None=Multiply(A=@n20) {constB=1.5}))
  n22 = Clamp(None=Subtract(A=OneMinus(None=@n21)) {constB=0.68})
  MaterialFunctionCall(Emissive=Multiply(A=Multiply(A=@n18, B=@n22) {constB=0.3}) {constB=8}) {materialFunction=&MF_FX_EmissiveAdjustments}
  MaterialFunctionCall(Normal=@n13.RGB, Flatness=Constant() {r=3}) {materialFunction=&FlattenNormal}
  n23 = Multiply(A=Time(), B=DynamicParameter().Param2) {constB=4}
  Panner(Coordinate=@n14, Time=Divide(A=Ceil(None=@n23)) {constB=4}) {speedY=0.4}
  ReflectionVectorWS()
  RotateAboutAxis(
   NormalizedRotationAxis=Constant3Vector() {constant=<rgba 0 0 1 0>}
   RotationAngle=Multiply(A=Time()) {constB=0.3}
   PivotPoint=ParticlePositionWS()
   Position=WorldPosition().XYZ
  )
  TextureSample() {texture=/Game/Environments/Props/Retail/Textures/T_DiscoBall_E}
  TextureSample() {samplerType=SAMPLERTYPE_Masks, texture=/Game/Environments/Props/Retail/Textures/T_DiscoBall_M}
  TextureSample() {samplerType=SAMPLERTYPE_Normal, texture=/Game/Environments/Props/Retail/Textures/T_DiscoBall_N}
 ]
}
