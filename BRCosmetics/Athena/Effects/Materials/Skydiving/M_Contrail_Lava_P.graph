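# Unlit translucent Niagara sprite material for a lava skydiving contrail, built from water splash textures.
# A panning noise map distorts the splash UVs, timed and scaled by dynamic parameters when UseDP is on; mobile skips distortion.
# Powered normal-map components plus squared splash green blend two lava colors, tinted by particle color for emissive.
# Opacity masks splash alpha with a hard threshold of its blue channel against particle alpha, giving an erosion cutoff.
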Material /BRCosmetics/Athena/Effects/Materials/Skydiving/M_Contrail_Lava_P
graph {
 properties {bUsedWithNiagaraSprites=true, blendMode=BLEND_Translucent, pixelDepthOffsetMode=PDOM_Legacy, shadingModel=MSM_Unlit, usedShadingModels=Unlit}
 nodes [
  Constant() {r=1}
  n1 = VectorParameter() {parameterName=Param_1, defaultValue=<rgba 0.6875 0.0342 0 1>}
  n2 = VectorParameter() {defaultValue=<rgba 1.1 0.4029 0.0338 1>}
  n3 = DynamicParameter() {paramNames [Time,Distort,Param3,Random]}
  n4 = StaticBoolParameter() {parameterName=UseDP}
  n5 = Panner(Coordinate=TextureCoordinate(), Time=StaticSwitch(True=Add(A=@n3.Time, B=@n3.Time), False=Time(), Value=@n4)) {speedY=-0.03}
  n6 = AppendVector(A=ScalarParameter() {defaultValue=1, parameterName=UVX}, B=ScalarParameter() {defaultValue=1, parameterName=UVY})
  n7 = ComponentMask(None=TextureSample(UVs=Multiply(A=@n5, B=@n6)) {texture=/Game/Global/Textures/Default/Tileable_Patterns/T_Tilling_Distortion_Map_01}.RGB) {g=true, r=true}
  n8 = StaticSwitch(True=@n3.Distort, False=Constant() {r=0.1}, Value=@n4)
  n9 = TextureCoordinate()
  n10 = ShadingPathSwitch(Default=Add(A=Multiply(A=@n7, B=@n8) {constB=0.1}, B=@n9), Mobile=@n9) {desc=Migrated from FeatureLevelSwitch}
  n11 = ComponentMask(None=TextureCoordinate()) {g=true}
  n12 = Abs(None=LinearInterpolate(
    A=Constant3Vector() {constant=<rgba 0 0 1 0>}
    B=TextureSample(UVs=@n10) {samplerType=SAMPLERTYPE_Normal, texture=/Game/Effects/Fort_Effects/Textures/Water/T_water_splash_04_n}.RGB
    Alpha=Ceil(None=@n11)
   ))
  n13 = Multiply(A=Power(
    Base=MaterialFunctionCall(f2=@n12) {materialFunction=&AddComponents}.f2
    Exp=Constant() {r=3}
   )) {constB=1.6}
  n14 = TextureSample(UVs=@n10) {desc=o, texture=/Game/Effects/Fort_Effects/Textures/Water/T_water_splash_04}
  n15 = Add(A=@n13, B=Multiply(A=@n14.G, B=@n14.G))
  n16 = ParticleColor()
  MaterialFunctionCall(Emissive=Multiply(A=LinearInterpolate(A=@n1.RGB, B=@n2.RGB, Alpha=@n15), B=@n16.RGB)) {materialFunction=&MF_FX_EmissiveAdjustments}
  Multiply(A=Constant3Vector() {constant=<rgba 3 0.6095 0.2202 0>}, B=@n13)
  n17 = Constant() {r=1}
  n18 = Constant()
  Multiply(
   A=Multiply(A=@n14.A, B=If(A=@n14.B, B=@n16.A, "A > B"=@n17, "A == B"=@n17, "A < B"=@n18))
   B=ScalarParameter() {defaultValue=1, parameterName=Alpha}
  )
 ]
}
