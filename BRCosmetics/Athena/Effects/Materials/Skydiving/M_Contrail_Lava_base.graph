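# Unlit translucent base material for a lava skydiving contrail, flagged for beam trails and Niagara ribbons.
# A sphere mask around an offset object position replaces ribbon U in panning, normal-distorted lava and smoke textures.
# Smoke and trail U set an erosion threshold; subtracting two time-offset threshold results makes the bright hot edge.
# Opacity combines erosion, UV mask and near-camera fade, adding depth fade outside mobile.
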Material /BRCosmetics/Athena/Effects/Materials/Skydiving/M_Contrail_Lava_base
graph {
 properties {
  bUsedWithBeamTrails=true
  bUsedWithNiagaraRibbons=true
  blendMode=BLEND_Translucent
  pixelDepthOffsetMode=PDOM_Legacy
  shadingModel=MSM_Unlit
  usedShadingModels=Unlit
 }
 nodes [
  n1 = ObjectPositionWS()
  n2 = VectorParameter() {parameterName=InitialOffset, defaultValue=<rgba 0 0 -1000 1>}
  n3 = Add(A=SphereMask(A=WorldPosition().XYZ, B=Add(A=@n1, B=@n2.RGB), Radius=Constant() {r=2550}, Hardness=Constant()), B=Constant() {r=-200})
  n4 = ComponentMask(None=TextureCoordinate()) {g=true}
  n5 = AppendVector(A=Multiply(A=@n3, B=Constant() {r=11}), B=Multiply(A=@n4, B=Constant() {r=0.5}))
  n6 = ComponentMask(None=TextureSample(UVs=Panner(Coordinate=@n5) {speedX=0.8}) {samplerType=SAMPLERTYPE_LinearColor, texture=/Game/Effects/Fort_Effects/Textures/Tile/T_Raygun_Muzzle_02}.RGB) {g=true}
  n7 = Constant3Vector() {constant=<rgba 0.4688 0.0655 0 0>}
  n8 = Multiply(A=@n4, B=Constant() {r=0.9})
  n9 = AppendVector(A=Multiply(A=@n3, B=Constant() {r=17}), B=@n8)
  n10 = ComponentMask(None=TextureSample(UVs=Panner(Coordinate=@n9) {speedX=1, speedY=0.1}) {samplerType=SAMPLERTYPE_Normal, texture=&T_Weap_Distort_N_02}.RGB) {g=true, r=true}
  n11 = Multiply(A=@n3, B=Constant() {r=26})
  n12 = Panner(Coordinate=Add(A=AppendVector(A=@n11, B=@n8), B=Multiply(A=@n10) {constB=0.2})) {speedX=3.6}
  n13 = TextureSample(UVs=@n12) {texture=/Game/Effects/Fort_Effects/Textures/T_Contrail_Lava2}
  n14 = Add(A=Multiply(A=@n10) {constB=0.11}, B=@n9)
  n15 = DynamicParameter()
  n16 = Add(A=Add(A=Constant() {r=0.025}, B=@n15.Param1), B=Time())
  n17 = TextureSample(UVs=Panner(Coordinate=@n9) {speedX=1.3}) {samplerType=SAMPLERTYPE_LinearColor, texture=/Game/Effects/Fort_Effects/Gameplay/SmokeDustExplosions/Textures/T_RibbonSmokeMasked_03}
  n18 = ComponentMask(None=TextureCoordinate()) {r=true}
  n19 = Multiply(A=OneMinus(None=@n17.R), B=Multiply(A=@n18, B=Constant() {r=0.5}))
  n20 = Add(A=Multiply(A=Power(Base=@n19) {constExponent=1.8}) {constB=2}, B=Power(Base=@n18) {constExponent=4.5})
  n21 = Constant() {r=1}
  n22 = Constant()
  n23 = If(
   A=TextureSample(UVs=Panner(Coordinate=@n14, Time=@n16) {speedX=2.6}) {texture=/Game/Effects/Fort_Effects/Textures/T_Contrail_Lava}.R
   B=@n20
   "A > B"=@n21
   "A == B"=@n21
   "A < B"=@n22
  )
  n24 = Panner(Coordinate=@n14, Time=Add(A=@n15.Param1, B=Time())) {speedX=2.6}
  n25 = Subtract(
   A=@n23
   B=If(A=TextureSample(UVs=@n24) {texture=&T_Contrail_Lava}.R, B=@n20, "A > B"=@n21, "A == B"=@n21, "A < B"=@n22)
  )
  n26 = Add(
   A=Multiply(A=@n13.R, B=Constant3Vector() {constant=<rgba 4 1.7951 0.0078 0>})
   B=Multiply(A=Clamp(None=@n25), B=Constant3Vector() {constant=<rgba 15 1.1909 0 0>}) {constB=40}
  )
  n27 = TextureSample(UVs=@n12) {texture=&T_Contrail_Lava}
  n28 = Add(
   A=Multiply(A=Constant3Vector() {constant=<rgba 3 0.2357 0 1>}, B=Subtract(A=@n27.A, B=@n27.A))
   B=Multiply(A=@n27.G, B=Constant3Vector() {constant=<rgba 1 0.1164 0.0267 1>})
  )
  n29 = Constant3Vector() {constant=<rgba 0.35 0.1 0.5 1>}
  n30 = Add(A=Multiply(A=Multiply(A=@n6, B=@n7)) {constB=1.28}, B=Subtract(A=Add(A=@n26, B=@n28), B=Multiply(A=@n29, B=@n13.G)))
  n31 = Multiply(A=@n30, B=Constant() {r=3})
  n32 = Multiply(A=@n30, B=Constant3Vector() {constant=<rgba 0.2865 0.0625 0.0522 0>})
  n33 = SphereMask(A=ComponentMask(None=TextureCoordinate()) {g=true}, B=Constant() {r=0.5}) {attenuationRadius=0.58, hardnessPercent=0}
  n34 = Multiply(A=@n18, B=Multiply(A=@n33, B=@n33))
  n35 = OneMinus(None=@n18)
  MaterialFunctionCall(Emissive=LinearInterpolate(
    A=Constant3Vector() {constant=<rgba 0.349 0.0818 0.0189 0>}
    B=LinearInterpolate(A=@n31, B=@n32, Alpha=@n34)
    Alpha=Power(Base=@n35) {constExponent=2.4}
   )) {materialFunction=&MF_FX_EmissiveAdjustments}
  n36 = SphereMask(A=TextureCoordinate(), B=Constant2Vector() {g=0.5, r=1.45}) {attenuationRadius=1.45, hardnessPercent=93}
  n37 = WorldPosition()
  n38 = CameraPositionWS()
  n39 = Constant() {r=85}
  n40 = Constant() {r=0.2}
  n41 = Multiply(A=Multiply(A=@n27.A, B=Multiply(A=@n36, B=@n23)), B=OneMinus(None=SphereMask(A=@n37.XYZ, B=@n38, Radius=@n39, Hardness=@n40)))
  ShadingPathSwitch(Default=Multiply(A=@n41, B=DepthFade()), Mobile=@n41) {desc=Migrated from FeatureLevelSwitch}
 ]
}
