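# Unlit translucent mesh-particle material for the lava skydiving contrail.
# Emissive combines an orange Fresnel-style rim with noise-tinted lava colors, blended bright to dark by dynamic Param2, through MF_FX_EmissiveAdjustments.
# An alpha term thresholds an inverted flare-burst texture against dynamic Param1 for erosion, scaled by particle alpha.
# A panning distortion normal and noise, multiplied by vertex normals, form a displacement-style vector; several other samples feed nothing.
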Material /BRCosmetics/Athena/Effects/Materials/Skydiving/M_Contrail_Lava_mesh
graph {
 properties {
  bUsedWithMeshParticles=true
  bUsedWithNiagaraMeshParticles=true
  blendMode=BLEND_Translucent
  pixelDepthOffsetMode=PDOM_Legacy
  previewMesh=/Game/Effects/Fort_Effects/Mesh/M_Contrail_Lava
  shadingModel=MSM_Unlit
  usedShadingModels=Unlit
 }
 nodes [
  Constant()
  n1 = CameraVectorWS()
  n2 = VertexNormalWS()
  n3 = Power(Base=OneMinus(None=DotProduct(A=@n1, B=@n2)))
  n4 = Multiply(A=Constant() {r=2}, B=Constant3Vector() {constant=<rgba 1.4 0.2938 0 0>})
  n5 = Constant3Vector() {constant=<rgba 1 0.1152 0 0>}
  n6 = TextureSample(UVs=TextureCoordinate() {uTiling=0.17, vTiling=0.3}) {texture=&T_NoisePack_01}
  n7 = Multiply(A=Constant() {r=1.2}, B=Add(A=Constant3Vector() {constant=<rgba 1 0.0695 0 0>}, B=Multiply(A=@n5, B=@n6.R)))
  n8 = Multiply(A=@n7, B=Constant3Vector() {constant=<rgba 0.2552 0.1004 0.0039 0>})
  n9 = DynamicParameter() {defaultValue=<rgba 0 1 1 1>}
  MaterialFunctionCall(Emissive=Add(A=Multiply(A=@n3, B=@n4), B=LinearInterpolate(A=@n7, B=@n8, Alpha=@n9.Param2))) {materialFunction=&MF_FX_EmissiveAdjustments}
  MaterialFunctionCall() {materialFunction=&DebugTimeSine}
  n10 = TextureSample() {texture=/Game/Effects/Fort_Effects/Materials/T_EFX_Flare_Burst_01}
  n11 = Constant() {r=1}
  Multiply(
   A=ParticleColor().A
   B=If(A=OneMinus(None=@n10.R), B=DynamicParameter() {defaultValue=<rgba 0 1 1 1>}.Param1, "A > B"=@n11, "A == B"=@n11, "A < B"=Constant())
  )
  n12 = Panner(Coordinate=TextureCoordinate() {uTiling=0.1, vTiling=0.1}) {speedY=1}
  n13 = TextureSample(UVs=@n12) {samplerType=SAMPLERTYPE_Normal, texture=&T_Weap_Distort_N_02}
  n14 = ComponentMask(None=@n13.RGB) {g=true, r=true}
  n15 = TextureSample(UVs=@n12) {texture=&T_NoisePack_01}
  n16 = VertexNormalWS()
  Multiply(A=Multiply(A=AppendVector(A=@n14, B=@n15.R), B=Multiply(A=@n16)), B=Constant() {r=5})
  TextureSample() {samplerType=SAMPLERTYPE_Normal, texture=/Game/Global/Textures/Default/Flowmaps/T_Smoke_FL_01}
  TextureSample(UVs=TextureCoordinate() {uTiling=0.2}) {texture=/Game/Effects/Fort_Effects/Textures/T_Contrail_Lava}
  Transform(None=@n13.RGB) {transformSourceType=TRANSFORMSOURCE_World, transformType=TRANSFORM_Tangent}
 ]
}
