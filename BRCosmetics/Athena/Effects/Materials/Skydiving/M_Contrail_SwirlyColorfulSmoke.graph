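# Unlit translucent material for swirly, colorful skydiving contrail smoke on beam trails and Niagara ribbons.
# Panned smoke, mask and swirl-noise textures, on sphere-mask world-position UVs or mesh UVs by switch, drive opacity and gradient or distance-based color.
# Edge masks, particle color, time-of-day multipliers, optional depth fade and view-angle fade shape the result, with negative desaturation boosting saturation.
# A distortion normal map builds wavy world position offset, masked near the object and zeroed on mobile.
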Material /BRCosmetics/Athena/Effects/Materials/Skydiving/M_Contrail_SwirlyColorfulSmoke
graph {
 properties {
  bUsedWithBeamTrails=true
  bUsedWithNiagaraRibbons=true
  blendMode=BLEND_Translucent
  pixelDepthOffsetMode=PDOM_Legacy
  shadingModel=MSM_Unlit
  translucencyLightingMode=TLM_VolumetricPerVertexDirectional
  usedShadingModels=Unlit
 }
 nodes [
  n1 = ParticleColor()
  n2 = ObjectPositionWS()
  n3 = VectorParameter() {parameterName=InitialOffset, defaultValue=<rgba 0 0 -1000 1>}
  n4 = Add(A=SphereMask(A=WorldPosition().XYZ, B=Add(A=@n2, B=@n3.RGB), Radius=Constant() {r=2550}, Hardness=Constant()), B=Constant() {r=-200}) {constB=-20}
  n5 = AppendVector(A=Multiply(A=@n4) {constB=21}, B=ComponentMask(None=TextureCoordinate()) {g=true})
  n6 = TextureSample(UVs=Panner(Coordinate=TextureCoordinate() {uTiling=0.52, vTiling=0.52}) {speedX=-0.75}) {samplerType=SAMPLERTYPE_Normal, texture=/Game/Global/Textures/Default/Flowmaps/T_SwirlFlow_02_N}
  n7 = Add(A=@n5, B=Multiply(A=ComponentMask(None=@n6.RGB) {g=true, r=true}) {constB=0.11})
  n8 = Multiply(A=DynamicParameter() {paramNames [WPO Speed Offset, WPO Tiling Offset, Time Mult, Not_Used]}.Time Mult, B=Time())
  n9 = MaterialFunctionCall(Float2=TextureCoordinate() {coordinateIndex=1}) {materialFunction=&BreakOutFloat2Components}
  n10 = ScalarParameter() {defaultValue=1, parameterName=UV_U}
  n11 = Panner(
   Coordinate=AppendVector(A=Multiply(A=@n9.R, B=@n10), B=@n9.G)
   Time=Time()
   Speed=AppendVector(A=ScalarParameter() {defaultValue=0.001, parameterName=SpeedX}, B=Constant())
  ) {speedX=0.001}
  n12 = TextureSampleParameter2D(UVs=StaticSwitchParameter(True=Panner(Coordinate=@n7, Time=@n8) {speedX=2}, False=@n11) {defaultValue=true, parameterName=UseWorldPositionUVs}) {parameterName=Texture, samplerType=SAMPLERTYPE_Grayscale, texture=&T_Smoke_Trail_Soft}
  n13 = Power(Base=@n12.R, Exp=ScalarParameter() {defaultValue=1.21, parameterName=Alpha Scope}) {constExponent=1.3}
  n14 = StaticSwitchParameter(True=Panner(Coordinate=@n5, Time=@n8) {speedX=2}, False=@n11) {defaultValue=true, parameterName=UseWorldPositionUVs}
  n15 = Multiply(A=Add(A=@n13, B=TextureSample(UVs=@n14) {texture=/Game/Effects/Fort_Effects/Textures/Smoke/T_Lobber_Trail_Mask}.RGB)) {constB=1.15}
  n16 = MaterialFunctionCall(Float2=TextureCoordinate()) {materialFunction=&BreakOutFloat2Components}
  n17 = Multiply(
   A=SphereMask(A=Power(Base=@n16.R) {constExponent=0.22}, B=Constant() {r=0.5}) {attenuationRadius=0.5, hardnessPercent=0}
   B=SphereMask(A=TextureCoordinate(), B=Constant2Vector() {g=0.5, r=1.1}) {attenuationRadius=1.1, hardnessPercent=50}
  )
  n18 = Constant() {r=2.1}
  n19 = Constant() {r=1}
  n20 = Multiply(
   A=@n1.A
   B=Multiply(
    A=Multiply(A=@n15, B=@n17)
    B=MaterialFunctionCall(Color_Morning=@n18, Color_Day=@n18, Color_Evening=@n19, Color_Night=@n19) {materialFunction=&Color_Time_of_Day}.Result
   )
  )
  n21 = StaticSwitchParameter(True=DepthFade(Opacity=@n20, FadeDistance=ScalarParameter() {defaultValue=30, parameterName=FadeDistance}), False=@n20) {parameterName=UseDepthFade}
  n22 = Constant() {r=10}
  LinearInterpolate(
   A=@n21
   B=Multiply(
    A=@n21
    B=MaterialFunctionCall("FadeSensitivity (optional)"=@n22) {materialFunction=&FadeBasedOnViewAngle}.Result
   )
   Alpha=ScalarParameter() {parameterName=MeshFede}
  )
  n23 = VectorParameter() {parameterName=ColorC, defaultValue=<rgba 0 0.9688 0.8727 1>}
  n24 = VectorParameter() {parameterName=ColorA, defaultValue=<rgba 0.1276 0.3217 0.7379 1>}
  n25 = MaterialFunctionCall(Float2=TextureCoordinate()) {materialFunction=&BreakOutFloat2Components}
  n26 = Multiply(A=@n25.R, B=@n25.R)
  n27 = VectorParameter() {parameterName=ColorB, defaultValue=<rgba 1 0.052 0.062 1>}
  n28 = VectorParameter() {parameterName=ColorA, defaultValue=<rgba 0.1276 0.3217 0.7379 1>}
  n29 = Power(
   Base=MaterialFunctionCall(Float2=TextureCoordinate()) {materialFunction=&BreakOutFloat2Components}.R
   Exp=ScalarParameter() {defaultValue=0.85, parameterName=Color Grad Pwr}
  )
  n30 = ScalarParameter() {defaultValue=0.85, parameterName=Contrast}
  n31 = SphereMask(A=WorldPosition().XYZ, B=ObjectPositionWS(), Radius=ScalarParameter() {defaultValue=850, parameterName=Color Range}, Hardness=Constant())
  n32 = ScalarParameter() {defaultValue=1.8, parameterName=Color Contrast}
  n33 = StaticSwitchParameter(
   True=LinearInterpolate(
    A=LinearInterpolate(A=@n23.RGB, B=@n24.RGB, Alpha=@n26)
    B=LinearInterpolate(A=@n27.RGB, B=@n28.RGB, Alpha=@n29)
    Alpha=Power(Base=@n12.RGB, Exp=@n30)
   )
   False=LinearInterpolate(
    A=Constant3Vector() {constant=<rgba 0.1276 0.3217 0.7379 1>}
    B=Constant3Vector() {constant=<rgba 0.8646 0.0097 0.2996 1>}
    Alpha=Power(Base=@n31, Exp=@n32)
   )
  ) {defaultValue=true, parameterName="WorldPositionColor "}
  n34 = Multiply(A=@n4) {constB=21.5}
  n35 = ComponentMask(None=TextureCoordinate()) {g=true}
  n36 = StaticSwitchParameter(True=Panner(Coordinate=AppendVector(A=@n34, B=@n35)) {speedX=1.5}, False=@n11) {defaultValue=true, parameterName=UseWorldPositionUVs}
  n37 = LinearInterpolate(
   A=Constant() {r=0.04}
   B=Constant() {r=1.35}
   Alpha=Add(
    A=TextureSampleParameter2D(UVs=@n36) {parameterName=NoiseTexture, samplerType=SAMPLERTYPE_LinearColor, texture=/Game/Global/Textures/Default/Flowmaps/T_SwirlFlow_02}.R
    B=Multiply(A=@n13) {constB=0.65}
   )
  ) {constA=0.5}
  n38 = ScalarParameter() {defaultValue=1.45, group=Time Of Day, parameterName=Morning/Day Mult}
  n39 = ScalarParameter() {defaultValue=1, group=Time Of Day, parameterName=Evening/Night Mult}
  n40 = Desaturation(
   None=Multiply(
    A=Multiply(A=@n33, B=@n37)
    B=MaterialFunctionCall(Color_Morning=@n38, Color_Day=@n38, Color_Evening=@n39, Color_Night=@n39) {materialFunction=&Color_Time_of_Day}.Result
   ) {constB=2.4}
   Fraction=ScalarParameter() {defaultValue=-1.75, parameterName=Saturation}
  )
  n41 = ScalarParameter() {parameterName=UseColor}
  MaterialFunctionCall(Emissive=Multiply(A=LinearInterpolate(A=@n40, B=@n37, Alpha=@n41), B=@n1.RGB)) {materialFunction=&MF_FX_EmissiveAdjustments}
  n42 = WorldPosition()
  n43 = DynamicParameter() {paramNames [WPO Speed Offset, WPO Tiling Offset, Time Mult, Not_Used]}
  n44 = Multiply(A=ObjectPositionWS(), B=Add(A=ScalarParameter() {defaultValue=0.9, group=WPO, parameterName=Speed}, B=@n43.WPO Speed Offset)) {constB=0.95}
  n45 = Divide(
   A=ComponentMask(None=Subtract(A=@n42.XYZ, B=@n44)) {b=true}
   B=Add(A=ScalarParameter() {defaultValue=510, group=WPO, parameterName=Tilling}, B=@n43.WPO Tiling Offset)
  )
  n46 = WorldPosition()
  n47 = Add(A=ObjectPositionWS(), B=VectorParameter() {parameterName=WPOMaskOffset, defaultValue=<rgba 0 0 0 1>}.RGB)
  n48 = ScalarParameter() {defaultValue=8000, group=WPO, parameterName=WPOmask}
  n49 = Constant()
  n50 = Multiply(
   A=Multiply(
    A=TextureSample(UVs=@n45) {samplerType=SAMPLERTYPE_Normal, texture=&T_Weap_Distort_N_02}.RGB
    B=ScalarParameter() {defaultValue=1000, group=WPO, parameterName=WavingAmount}
   ) {constB=1500}
   B=OneMinus(None=SphereMask(A=@n46.XYZ, B=@n47, Radius=@n48, Hardness=@n49))
  )
  NamedRerouteDeclaration(None=@n50) {name=Test, nodeColor=<rgba 1 0.0469 0 1>}
  NamedRerouteUsage()
  n51 = Constant()
  Reroute(None=LinearInterpolate(
    A=ShadingPathSwitch(Default=@n50, Mobile=@n51) {desc=Migrated from FeatureLevelSwitch}
    B=Constant()
    Alpha=ScalarParameter() {parameterName=UseWPO}
   ))
 ]
}
