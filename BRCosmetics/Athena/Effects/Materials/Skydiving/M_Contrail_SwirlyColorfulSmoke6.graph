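# Translucent skydiving contrail material for beam and Niagara ribbon trails, drawing swirly, multicolored animated smoke.
# A panned smoke texture, with U driven by object distance, becomes cycling bands through channel-packed phase math and a ramp lookup.
# Those bands blend three color parameters and, with UV edge masks and particle alpha, shape opacity.
# A world-space normal-map sample drives waving offset, suppressed near the object and zeroed on mobile.
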Material /BRCosmetics/Athena/Effects/Materials/Skydiving/M_Contrail_SwirlyColorfulSmoke6
graph {
 properties {
  bUsedWithBeamTrails=true
  bUsedWithNiagaraRibbons=true
  blendMode=BLEND_Translucent
  pixelDepthOffsetMode=PDOM_Legacy
  usedShadingModels=Default Lit
 }
 nodes [
  n1 = SphereMask(
   A=WorldPosition().XYZ
   B=Add(A=ObjectPositionWS(), B=VectorParameter() {parameterName=InitialOffset, defaultValue=<rgba 0 0 -1000 1>}.RGB)
   Radius=Constant() {r=2550}
   Hardness=Constant()
  )
  n2 = Multiply(A=@n1, B=@n1)
  n3 = Constant() {r=-200}
  n4 = TextureCoordinate()
  n5 = AppendVector(A=Multiply(A=Add(A=@n2, B=@n3) {constB=-20}) {constB=25}, B=Multiply(A=ComponentMask(None=@n4) {g=true}) {constB=0.68})
  n6 = DynamicParameter() {defaultValue=<rgba 0 1 1 1>}
  n7 = Constant()
  n8 = TextureSample(UVs=Panner(Coordinate=Add(A=@n5), Time=Time(), Speed=AppendVector(A=@n6.Param2, B=@n7))) {samplerType=SAMPLERTYPE_LinearColor, texture=/BRCosmetics/Effects/Fort_Effects/Textures/T_Swirlysmoke_a}
  n9 = Multiply(A=Add(A=DynamicParameter() {defaultValue=<rgba 0 1 1 1>}.Param1, B=Time()), B=ScalarParameter() {defaultValue=0.5}) {constB=-0.28}
  n10 = LinearInterpolate(Alpha=@n8.R) {constA=0.5, constB=0.75}
  n11 = Add(A=Add(A=@n8.B, B=Multiply(A=@n9, B=@n10)), B=Multiply(A=@n8.R) {constB=5})
  n12 = Multiply(A=Frac(None=Multiply(A=@n11) {constB=0.3})) {constB=2}
  n13 = Frac(None=Saturate(None=Subtract(A=@n12) {constB=0.5}))
  n14 = Multiply(
   A=@n8.G
   B=Power(Base=TextureSample(UVs=@n13) {samplerSource=SSM_Clamp_WorldGroupSettings, texture=/Game/Effects/Fort_Effects/Textures/Ramps/T_Power_Ball_Beam_Ramp}.R)
  )
  LinearInterpolate(A=Constant3Vector() {constant=<rgba 0.141 0.1257 0.7379 1>}, B=Constant3Vector() {constant=<rgba 0 0.767 2 1>}, Alpha=@n14)
  n15 = ParticleColor()
  n16 = MaterialFunctionCall(Float2=TextureCoordinate()) {materialFunction=&BreakOutFloat2Components}
  n17 = LinearInterpolate(
   A=VectorParameter() {parameterName=ColorC, defaultValue=<rgba 0 0.5398 0.9688 1>}.RGB
   B=VectorParameter() {parameterName=ColorA, defaultValue=<rgba 0.0175 0.1352 0.7379 1>}.RGB
   Alpha=Multiply(A=@n16.R, B=@n16.R)
  )
  n18 = MaterialFunctionCall(Float2=TextureCoordinate()) {materialFunction=&BreakOutFloat2Components}
  n19 = LinearInterpolate(
   A=VectorParameter() {parameterName=ColorB, defaultValue=<rgba 1 0.052 0.062 1>}.RGB
   B=VectorParameter() {parameterName=ColorA, defaultValue=<rgba 0.0175 0.1352 0.7379 1>}.RGB
   Alpha=SquareRoot(None=@n18.R)
  )
  MaterialFunctionCall(Emissive=Multiply(A=LinearInterpolate(A=@n17, B=@n19, Alpha=@n14), B=@n15.RGB)) {materialFunction=&MF_FX_EmissiveAdjustments}
  n20 = MaterialFunctionCall(Float2=TextureCoordinate()) {materialFunction=&BreakOutFloat2Components}
  n21 = Multiply(
   A=SphereMask(A=Power(Base=@n20.R) {constExponent=0.22}, B=Constant() {r=0.5}) {attenuationRadius=0.5, hardnessPercent=0}
   B=SphereMask(A=TextureCoordinate(), B=Constant2Vector() {g=0.5, r=1.1}) {attenuationRadius=1.05, hardnessPercent=88}
  )
  n22 = SphereMask(
   A=MaterialFunctionCall(Float2=TextureCoordinate()) {materialFunction=&BreakOutFloat2Components}.G
   B=Constant() {r=0.5}
  ) {attenuationRadius=0.45, hardnessPercent=0}
  n23 = ScalarParameter() {defaultValue=0.8, parameterName=Alphamaster}
  Multiply(A=@n15.A, B=Multiply(A=Multiply(A=@n14, B=@n23) {constB=0.2}, B=Multiply(A=@n21, B=@n22)))
  Power(Base=@n18.R, Exp=ScalarParameter() {defaultValue=0.5})
  n24 = ScalarParameter() {defaultValue=0.9, parameterName=Speed}
  n25 = DynamicParameter() {defaultValue=<rgba 0 1 1 1>}
  n26 = Subtract(A=WorldPosition().XYZ, B=Multiply(A=ObjectPositionWS(), B=Add(A=@n24, B=@n25.Param1)) {constB=0.95})
  n27 = ScalarParameter() {defaultValue=510, parameterName=Tilling}
  n28 = TextureSample(UVs=Divide(A=ComponentMask(None=@n26) {b=true}, B=Add(A=@n27, B=@n25.Param2))) {samplerType=SAMPLERTYPE_Normal, texture=&T_Weap_Distort_N_02}
  n29 = ScalarParameter() {defaultValue=700, parameterName=WavingAmount}
  n30 = SphereMask(
   A=WorldPosition().XYZ
   B=Add(A=ObjectPositionWS(), B=VectorParameter() {parameterName=WPOMaskOffset, defaultValue=<rgba 0 0 0 1>}.RGB)
   Radius=ScalarParameter() {defaultValue=8000, parameterName=WPOmask}
   Hardness=Constant()
  )
  ShadingPathSwitch(Default=Multiply(A=Multiply(A=@n28.RGB, B=@n29) {constB=1500}, B=OneMinus(None=@n30)), Mobile=Constant()) {desc=Migrated from FeatureLevelSwitch}
 ]
}
