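# Additive unlit particle material for snowflake skydiving contrail sprites, usable with Niagara and Cascade.
# Particle color drives emissive through MF_FX_EmissiveAdjustments, while particle alpha multiplies a cross-flare texture's alpha and a time-based sine flicker.
# A static switch picks plain UVs or center scaling driven by a dynamic parameter.
# The panner's time is a constant, so the Speed X/Y parameters act as a fixed UV offset on the clamped texture.
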Material /BRCosmetics/Athena/Effects/Materials/Skydiving/M_Contrail_snowflakes02
graph {
 properties {
  bUsedWithNiagaraSprites=true
  bUsedWithParticleSprites=true
  blendMode=BLEND_Additive
  pixelDepthOffsetMode=PDOM_Legacy
  shadingModel=MSM_Unlit
  usedShadingModels=Unlit
 }
 nodes [
  n1 = ParticleColor()
  MaterialFunctionCall(Emissive=Multiply(A=Constant() {r=1}, B=@n1.RGB)) {materialFunction=&MF_FX_EmissiveAdjustments}
  n2 = TextureCoordinate()
  n3 = MaterialFunctionCall(UVs=TextureCoordinate(), Texture Scale=DynamicParameter() {defaultValue=<rgba 0.4 1 1 1>}.Param1) {materialFunction=&ScaleUVsByCenter}
  n4 = ScalarParameter() {defaultValue=0.42, parameterName=Speed X}
  n5 = ScalarParameter() {defaultValue=0.22, parameterName=Speed Y}
  n6 = TextureSample(UVs=Panner(
    Coordinate=StaticSwitchParameter(True=@n2, False=@n3.UVs) {parameterName=Texture UV}
    Time=Constant() {r=1}
    Speed=AppendVector(A=@n4, B=@n5)
   ) {speedX=0.42, speedY=0.22}) {
   samplerSource=SSM_Clamp_WorldGroupSettings
   samplerType=SAMPLERTYPE_Grayscale
   texture=/Game/Effects/Fort_Effects/Textures/Flares/T_Cross_Flare
  }
  n7 = Constant() {r=0.5}
  n8 = Constant() {r=1}
  n9 = Sine(None=Multiply(A=Time()) {constB=0.7})
  Multiply(A=@n1.A, B=Multiply(A=@n6.A, B=LinearInterpolate(A=@n7, B=@n8, Alpha=@n9)))
 ]
}
