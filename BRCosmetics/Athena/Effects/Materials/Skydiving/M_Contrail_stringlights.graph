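# Unlit mesh-particle material for a skydiving contrail of string lights, emitting colored bulb glow with a wavy vertex offset.
# A static switch picks either four colors chosen by stepped sines of a dynamic parameter, or time-blinking colors masked by vertex-color channels with optional green.
# Fresnel (skipped on mobile), dynamic parameters and per-shading-path brightness scale the emissive.
# A height-based, time-driven sine along local X, sphere-masked around the particle, creates the sway.
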Material /BRCosmetics/Athena/Effects/Materials/Skydiving/M_Contrail_stringlights
graph {
 properties {
  bUsedWithMeshParticles=true
  bUsedWithNiagaraMeshParticles=true
  pixelDepthOffsetMode=PDOM_Legacy
  shadingModel=MSM_Unlit
  usedShadingModels=Unlit
 }
 nodes [
  n1 = DynamicParameter()
  n2 = ScalarParameter() {defaultValue=4, parameterName=light}
  n3 = ScalarParameter() {defaultValue=6, parameterName=light Low}
  n4 = ScalarParameter() {defaultValue=40, parameterName=Fresnel light}
  n5 = DynamicParameter()
  n6 = LinearInterpolate(
   A=VectorParameter() {parameterName=Color_a, defaultValue=<rgba 0.0208 0.0145 0 1>}.RGB
   B=VectorParameter() {parameterName=Color_b, defaultValue=<rgba 0.026 0.0003 0.0052 1>}.RGB
   Alpha=Ceil(None=Sine(None=@n5.Param1))
  )
  n7 = Add(A=Constant() {r=0.3}, B=@n5.Param1)
  n8 = LinearInterpolate(
   A=VectorParameter() {parameterName=Color_c, defaultValue=<rgba 0.0005 0.0235 0.0365 1>}.RGB
   B=VectorParameter() {parameterName=Color_d, defaultValue=<rgba 0.0029 0.0417 0 1>}.RGB
   Alpha=Ceil(None=Sine(None=@n7))
  )
  n9 = Constant() {r=0.15}
  n10 = Ceil(None=Sine(None=Add(A=@n9, B=@n5.Param1)))
  n11 = Sine(None=Multiply(A=Time(), B=Constant() {r=2}))
  n12 = Multiply(A=Clamp(None=Ceil(None=@n11)) {minDefault=0.018}, B=Constant3Vector() {constant=<rgba 1 0.3562 0 0>})
  n13 = VertexColor()
  n14 = Time()
  n15 = Constant() {r=3.6}
  n16 = Sine(None=Multiply(A=Subtract(A=@n14, B=@n15), B=Constant() {r=1}))
  n17 = Multiply(A=Clamp(None=Ceil(None=@n16)) {minDefault=0.018}, B=Constant3Vector() {constant=<rgba 0.6927 0.0078 0.1396 0>})
  n18 = Multiply(A=Subtract(A=Time(), B=Constant() {r=4.5}), B=Constant() {r=1.4})
  n19 = Clamp(None=Ceil(None=Sine(None=@n18))) {minDefault=0.018}
  n20 = Constant3Vector() {constant=<rgba 0.012 0.573 0.8906 0>}
  n21 = Add(A=Add(A=Multiply(A=@n12, B=@n13.R), B=Multiply(A=@n13.B, B=@n17)), B=Multiply(A=Multiply(A=@n19, B=@n20), B=VertexColor().G))
  n22 = Multiply(A=Subtract(A=Time(), B=Constant() {r=4}), B=Constant() {r=1.75})
  n23 = Clamp(None=Ceil(None=Sine(None=@n22))) {maxDefault=0.45, minDefault=0.018}
  n24 = Constant3Vector() {constant=<rgba 0.0702 1 0 0>}
  n25 = StaticSwitchParameter(True=Multiply(A=Multiply(A=@n23, B=@n24), B=VertexColor().A), False=Constant()) {parameterName=Green Switch}
  n26 = Desaturation(
   None=StaticSwitchParameter(True=LinearInterpolate(A=@n6, B=@n8, Alpha=@n10), False=Add(A=@n21, B=@n25)) {parameterName=Flashing or fixing}
   Fraction=Constant() {r=-0.5}
  )
  n27 = Multiply(A=Multiply(A=Fresnel() {baseReflectFraction=0, exponent=3.9}, B=Multiply(A=@n1.Param2, B=@n4)), B=@n26)
  n28 = Multiply(
   A=Multiply(A=@n1.Param4, B=ShadingPathSwitch(Default=@n2, Deferred=@n2, Forward=@n3, Mobile=@n3))
   B=ShadingPathSwitch(Default=Add(A=@n27, B=@n26), Mobile=@n26) {desc=Migrated from FeatureLevelSwitch}
  )
  MaterialFunctionCall(Emissive=Desaturation(None=Add(B=@n28) {constA=0.05}, Fraction=ScalarParameter() {defaultValue=-1, parameterName=saturation})) {materialFunction=&MF_FX_EmissiveAdjustments}
  n29 = WorldPosition()
  n30 = ParticlePositionWS()
  n31 = ComponentMask(None=Subtract(A=Add(A=@n29.XYZ) {constB=100}, B=@n30)) {b=true}
  n32 = DynamicParameter() {parameterIndex=1, defaultValue=<rgba 0 -2.2 200 8>, paramNames [Offset,Speed,UVScale,WPO]}
  n33 = Add(A=Time(), B=@n32.Offset)
  n34 = Sine(None=Add(A=Divide(A=@n31, B=@n32.UVScale) {constB=200}, B=Multiply(A=@n33, B=@n32.Speed) {constB=-2.2}))
  n35 = Constant3Vector() {constant=<rgba 0 0 1 0>}
  n36 = ScalarParameter() {defaultValue=200, parameterName=MaskOffset}
  n37 = SphereMask(A=WorldPosition().XYZ, B=Add(A=@n30, B=Multiply(A=@n35, B=@n36)) {constB=70}) {attenuationRadius=200, hardnessPercent=50}
  Multiply(
   A=Transform(None=Constant3Vector() {constant=<rgba 1 0 0 0>}) {transformSourceType=TRANSFORMSOURCE_Local}
   B=Multiply(A=Multiply(A=@n34, B=@n37), B=@n32.WPO) {constB=8}
  )
 ]
}
