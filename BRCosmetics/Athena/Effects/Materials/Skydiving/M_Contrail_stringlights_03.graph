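# Translucent, two-sided lit material for mesh and Niagara mesh particles, apparently string lights in a skydiving contrail.
# A fake spherical normal from UVs gives directional shading, blended toward BaseColor by DayLight, with mobile using plain color; BaseColor also feeds an emissive adjustment function.
# A time-animated sine along world Z, sphere-masked around an offset from the particle position, sways vertices along local X.
# Dynamic parameters drive the wave's offset, speed, scale and amplitude.
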Material /BRCosmetics/Athena/Effects/Materials/Skydiving/M_Contrail_stringlights_03
graph {
 properties {
  bUsedWithMeshParticles=true
  bUsedWithNiagaraMeshParticles=true
  blendMode=BLEND_Translucent
  pixelDepthOffsetMode=PDOM_Legacy
  twoSided=true
  usedShadingModels=Default Lit
 }
 nodes [
  n1 = TextureCoordinate()
  n2 = Constant() {r=0.5}
  n3 = Multiply(A=Subtract(A=@n1) {constB=0.5}, B=SphereMask(A=@n1, B=@n2, Radius=@n2, Hardness=Constant()))
  n4 = Normalize(VectorInput=AppendVector(A=Multiply(A=@n3) {constB=3.5}, B=Constant() {r=1}))
  n5 = AppendVector(A=Constant() {r=1}, B=Constant() {r=-1})
  n6 = Constant() {r=1}
  n7 = DotProduct(A=Saturate(None=TransformPosition(None=@n4) {transformType=TRANSFORMPOSSOURCE_World}), B=Normalize(VectorInput=AppendVector(A=@n5, B=@n6)))
  n8 = VectorParameter() {parameterName=BaseColor, defaultValue=<rgba 0 0 0 1>}
  LinearInterpolate(
   A=ShadingPathSwitch(Default=Multiply(A=@n7) {constB=0.14}, Mobile=@n8.RGB) {desc=Migrated from FeatureLevelSwitch}
   B=@n8.RGB
   Alpha=ScalarParameter() {defaultValue=0.9, parameterName=DayLight}
  )
  MaterialFunctionCall(Emissive=@n8.RGB) {materialFunction=&MF_FX_EmissiveAdjustments}
  n9 = WorldPosition()
  n10 = ParticlePositionWS()
  n11 = ComponentMask(None=Subtract(A=Add(A=@n9.XYZ) {constB=100}, B=@n10)) {b=true}
  n12 = DynamicParameter() {parameterIndex=1, defaultValue=<rgba 0 -2.2 200 8>, paramNames [Offset,Speed,UVScale,WPO]}
  n13 = Add(A=Time(), B=@n12.Offset)
  n14 = Sine(None=Add(A=Divide(A=@n11, B=@n12.UVScale) {constB=200}, B=Multiply(A=@n13, B=@n12.Speed) {constB=-2.2}))
  n15 = Constant3Vector() {constant=<rgba 0 0 1 0>}
  n16 = ScalarParameter() {defaultValue=200, parameterName=MaskOffset}
  n17 = SphereMask(A=WorldPosition().XYZ, B=Add(A=@n10, B=Multiply(A=@n15, B=@n16)) {constB=70}) {attenuationRadius=200, hardnessPercent=50}
  Multiply(
   A=Transform(None=Constant3Vector() {constant=<rgba 1 0 0 0>}) {transformSourceType=TRANSFORMSOURCE_Local}
   B=Multiply(A=Multiply(A=@n14, B=@n17), B=@n12.WPO) {constB=8}
  ) {constA=1}
  Power(Base=MaterialFunctionCall(Float2=TextureCoordinate()) {materialFunction=&BreakOutFloat2Components}.G) {constExponent=1.5}
  ScalarParameter() {parameterName=Metallic}
  ScalarParameter() {parameterName=Roughness}
 ]
}
