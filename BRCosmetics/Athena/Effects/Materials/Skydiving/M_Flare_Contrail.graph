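# Unlit, two-sided translucent particle material for a skydiving flare contrail, usable on Niagara and Cascade sprites.
# Particle color RGB runs through a shared FX emissive adjustment function.
# A mask multiplies particle alpha, a camera flare texture's red channel and a double-sided Fresnel term with a tunable exponent, shaping the flare by viewing angle.
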Material /BRCosmetics/Athena/Effects/Materials/Skydiving/M_Flare_Contrail
graph {
 properties {
  bUsedWithNiagaraSprites=true
  bUsedWithParticleSprites=true
  blendMode=BLEND_Translucent
  pixelDepthOffsetMode=PDOM_Legacy
  shadingModel=MSM_Unlit
  twoSided=true
  usedShadingModels=Unlit
 }
 nodes [
  n1 = ParticleColor()
  MaterialFunctionCall(Emissive=@n1.RGB) {materialFunction=&MF_FX_EmissiveAdjustments}
  Multiply(
   A=Multiply(A=@n1.A, B=TextureSample() {texture=/Game/Environments/FrontEnd/Textures/T_EFX_CameraFlare_02}.R)
   B=MaterialFunctionCall(Exponent=ScalarParameter() {defaultValue=3, parameterName=FresnelExponent}) {materialFunction=&MF_Fresnel_FX}.Fresnel Double Sided
  )
 ]
}
