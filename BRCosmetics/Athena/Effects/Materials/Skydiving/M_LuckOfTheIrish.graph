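# Translucent particle material for the Luck of the Irish skydiving effect, drawing a clover leaf texture on Cascade or Niagara sprites.
# A panning normal map offsets the UVs for wobble, which the Noise parameter fades out as it rises.
# Texture color, tinted by particle color, blends toward Emissive Color before an emissive adjustment function.
# The opacity mask blends between plain and SubUV samples, then multiplies by particle alpha.
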Material /BRCosmetics/Athena/Effects/Materials/Skydiving/M_LuckOfTheIrish
graph {
 properties {
  bUsedWithNiagaraSprites=true
  bUsedWithParticleSprites=true
  blendMode=BLEND_Translucent
  pixelDepthOffsetMode=PDOM_Legacy
  usedShadingModels=Default Lit
 }
 nodes [
  n1 = TextureCoordinate() {uTiling=0.5, vTiling=0.42}
  n2 = ComponentMask(None=TextureSample(UVs=Panner(Coordinate=@n1) {speedX=1}) {samplerType=SAMPLERTYPE_Normal, texture=&T_Weap_Distort_N_02}.RGB) {g=true, r=true}
  n3 = Add(A=TextureCoordinate(), B=LinearInterpolate(A=Multiply(A=@n2) {constB=0.06}, B=Constant(), Alpha=ScalarParameter() {parameterName=Noise}))
  n4 = TextureSampleParameter2D(UVs=@n3) {parameterName=Texture, texture=/BRCosmetics/Packages/Fortress_Transylvania/Transylvania_effects/Textures/T_LeafClover}
  n5 = Multiply(A=@n4.RGB) {constB=1.6}
  n6 = ParticleColor()
  MaterialFunctionCall(Emissive=LinearInterpolate(
    A=Multiply(A=@n5, B=@n6.RGB)
    B=VectorParameter() {parameterName=Emissive Color, defaultValue=<rgba 1 1 1 1>}.RGB
    Alpha=ScalarParameter() {parameterName=Emissive Color Swicth}
   )) {materialFunction=&MF_FX_EmissiveAdjustments}
  n7 = TextureSampleParameterSubUV!(UVs=@n3, Apply View MipBias) {parameterName=SubUV, texture=&T_LeafClover}
  Multiply(
   A=LinearInterpolate(
    A=StaticComponentMaskParameter(None=@n4.RGBA) {defaultA=true, parameterName=Opacity}
    B=StaticComponentMaskParameter(None=@n7.RGBA) {defaultA=true, parameterName=Opacity}
    Alpha=ScalarParameter() {parameterName=SubUV}
   )
   B=@n6.A
  )
 ]
}
