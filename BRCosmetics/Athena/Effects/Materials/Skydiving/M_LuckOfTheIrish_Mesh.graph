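# Masked, two-sided lit material for four-leaf clover mesh particles in the Luck of the Irish skydiving effect.
# Clover color is tinted by time of day, adjusted by facing side, then added to a switchable green Fresnel glow within a sphere mask offset from particle position.
# Switchable world position offset uses scrolling noise driven by time, UVs and dynamic parameters, scaled by vertex color and suppressed near the object's center.
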Material /BRCosmetics/Athena/Effects/Materials/Skydiving/M_LuckOfTheIrish_Mesh
graph {
 properties {
  bUsedWithMeshParticles=true
  bUsedWithNiagaraMeshParticles=true
  blendMode=BLEND_Masked
  pixelDepthOffsetMode=PDOM_Legacy
  twoSided=true
  usedShadingModels=Default Lit
 }
 nodes [
  n1 = Multiply(
   A=ScalarParameter() {defaultValue=21.5, parameterName=Day_light}
   B=VectorParameter() {parameterName=Day_light_color, defaultValue=<rgba 0.1849 0.7396 0.0179 1>}.RGB
  )
  n2 = MaterialFunctionCall(Float3=ParticlePositionWS()) {materialFunction=&BreakOutFloat3Components}
  n3 = AppendVector(A=Subtract(A=@n2.R, B=DynamicParameter().Param2) {constB=18}, B=@n2.G)
  n4 = Multiply(
   A=SphereMask(
    A=WorldPosition().XYZ
    B=AppendVector(A=@n3, B=@n2.B)
    Radius=ScalarParameter() {defaultValue=32, parameterName=Day light Mask size}
    Hardness=Constant()
   )
   B=Fresnel() {baseReflectFraction=0.02, exponent=1.4}
  )
  n5 = VectorParameter() {parameterName=Color_evening, defaultValue=<rgba 0.6003 1 0.0812 1>}
  n6 = VectorParameter() {parameterName=Color_day, defaultValue=<rgba 1 1 1 1>}
  n7 = ScalarParameter() {defaultValue=1.25, parameterName=Two sign light}
  n8 = TextureSample() {texture=/BRCosmetics/Effects/Fort_Effects/Textures/T_LeafClover_mesh}
  n9 = Multiply(
   A=MaterialFunctionCall(Color_Morning=@n5.RGB, Color_Day=@n6.RGB, Color_Evening=@n5.RGB, Color_Night=@n6.RGB) {materialFunction=&Color_Time_of_Day}.Result
   B=LinearInterpolate(A=Multiply(A=@n7, B=@n8.RGB), B=@n8.RGB, Alpha=TwoSidedSign())
  )
  Add(
   A=StaticSwitchParameter(True=Constant(), False=Multiply(A=@n1, B=@n4)) {parameterName=Day light}
   B=Desaturation(None=Multiply(A=@n9) {constB=0.6}, Fraction=ScalarParameter() {defaultValue=-0.15, parameterName=Desaturation})
  )
  Constant() {r=0.2}
  n10 = DynamicParameter()
  n11 = Multiply(A=Time(), B=Add(A=ScalarParameter() {defaultValue=180, parameterName=WPO_Speed}, B=@n10.Param2)) {constB=240}
  n12 = TextureCoordinate() {coordinateIndex=2}
  n13 = ScalarParameter() {defaultValue=48, parameterName=WPO_UV}
  n14 = TextureCoordinate()
  n15 = Add(
   A=Multiply(A=ComponentMask(None=@n12) {g=true}, B=Add(A=@n13, B=@n10.Param1)) {constB=80}
   B=Multiply(A=ComponentMask(None=@n14) {g=true}) {constB=0}
  )
  n16 = TextureSampleParameter2D(UVs=Divide(A=Add(A=@n11, B=@n15), B=ScalarParameter() {defaultValue=400, parameterName=Tilling})) {parameterName=T_WPO, texture=&T_NoisePack_01}
  n17 = ComponentMask(None=@n16.RGB) {g=true, r=true}
  n18 = Multiply(A=@n16.B, B=ScalarParameter() {defaultValue=0.3, parameterName=WPO_G}) {constB=0.3}
  n19 = Multiply(A=VertexColor(), B=Multiply(A=AppendVector(A=@n17, B=@n18), B=ScalarParameter() {defaultValue=5, parameterName=WPO_intensity}) {constB=17})
  n20 = OneMinus(None=SphereMask(A=WorldPosition().XYZ, B=ObjectPositionWS(), Radius=ScalarParameter() {defaultValue=65, parameterName=Sphere Msk size}, Hardness=Constant()))
  n21 = Multiply(A=@n20, B=@n20)
  StaticSwitchParameter(
   True=Constant()
   False=MaterialFunctionCall!(InVector=Multiply(A=@n19, B=@n21)) {materialFunction=/Game/Athena/Environments/Landscape/Foliage/MF_MakePrecise}.Result
  ) {parameterName=Use WPO}
  TextureSample() {samplerType=SAMPLERTYPE_Normal, texture=/Game/Missions/Quests/ReactiveQuests/Events/Spring2018/Fetch/Props/Textures/T_four_leaf_clover_N}
  TextureSample() {samplerType=SAMPLERTYPE_Masks, texture=/Game/Missions/Quests/ReactiveQuests/Events/Spring2018/Fetch/Props/Textures/T_four_leaf_clover_S}
 ]
}
