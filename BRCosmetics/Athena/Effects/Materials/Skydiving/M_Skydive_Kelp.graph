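# Translucent, unlit, two-sided Niagara ribbon/sprite material that draws wiggling green kelp strands, apparently for a skydive trail cosmetic.
# Panned sine/cosine waves, plus a normal-map detail on non-mobile, distort UVs sampling a gradient mask, with dynamic parameters driving offset, speed and direction.
# Smoothstepped masks and panning caustics, skipped on mobile, shape color and opacity.
# Pinch, depth and optional camera fades soften the strand.
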Material /BRCosmetics/Athena/Effects/Materials/Skydiving/M_Skydive_Kelp
graph {
 properties {
  bUsedWithNiagaraRibbons=true
  bUsedWithNiagaraSprites=true
  bUsedWithParticleSprites=true
  blendMode=BLEND_Translucent
  pixelDepthOffsetMode=PDOM_Legacy
  shadingModel=MSM_Unlit
  twoSided=true
  usedShadingModels=Unlit
 }
 nodes [
  n1 = DynamicParameter() {defaultValue=<rgba 1 0.56 1 0.1>, paramNames ["offset(0-1)", speed, "reverse(0,1)", SeaweedWiggle]}
  n2 = TextureCoordinate()
  n3 = ComponentMask(None=@n2) {r=true}
  n4 = AppendVector(A=Abs(None=Subtract(A=@n1."reverse(0,1)", B=@n3)), B=ComponentMask(None=@n2) {g=true})
  n5 = ComponentMask(None=@n4) {r=true}
  n6 = Add(A=@n4, B=AppendVector(A=@n1."offset(0-1)", B=Constant()))
  n7 = Add(A=Multiply(A=Time(), B=@n1.speed) {constB=10})
  n8 = Frac(None=ComponentMask(None=Panner(Coordinate=@n6, Time=@n7) {speedX=0.3}) {r=true})
  n9 = ScalarParameter() {defaultValue=1, group=Distort_map, parameterName=Disort_frequency, sortPriority=4}
  n10 = Panner(Coordinate=Add(A=@n6, B=Constant2Vector() {r=0.2}), Time=@n7) {speedX=0.3}
  n11 = Multiply(A=Frac(None=ComponentMask(None=@n10) {r=true}), B=@n9)
  n12 = Add(A=Sine(None=Multiply(A=@n8, B=@n9)) {period=0.5}, B=Multiply(A=Cosine(None=@n11)) {constB=0.1})
  n13 = Constant2Vector() {g=0.25, r=1}
  n14 = TextureSampleParameter2D(UVs=Panner(Coordinate=Multiply(A=@n4, B=@n13), Time=Time()) {speedX=2, speedY=0.1}) {
   group=Distort_map
   parameterName=Detail_disort_map
   samplerType=SAMPLERTYPE_Normal
   texture=/Game/Global/Textures/Default/Tileable_Patterns/1Normal
  }
  n15 = Multiply(A=ComponentMask(None=@n14.RGB) {g=true, r=true}, B=Constant2Vector() {g=0.75, r=0.5})
  n16 = Multiply(A=@n5, B=ShadingPathSwitch(Default=Add(A=@n12, B=@n15) {constB=0}, Mobile=@n12) {desc=Migrated from FeatureLevelSwitch})
  n17 = ScalarParameter() {defaultValue=0.2, group=Distort_map, parameterName=Disort_amplitude, sortPriority=5}
  n18 = ComponentMask(None=TextureCoordinate()) {r=true}
  n19 = Add(A=@n4, B=Multiply(A=Multiply(A=@n16, B=@n17) {constB=0.15}, B=Power(Base=@n18) {constExponent=0.5}))
  n20 = ComponentMask(None=@n19) {r=true}
  n21 = ComponentMask(None=TextureCoordinate() {uTiling=0, unMirrorU=true, vTiling=0}) {r=true}
  AppendVector(A=@n20, B=@n21)
  Divide(A=ComponentMask(None=WorldPosition().XYZ) {b=true}, B=Divide(A=Constant() {r=-1618}) {constB=4}) {constB=1618}
  n22 = LinearInterpolate(
   A=Constant3Vector() {constant=<rgba 1 1 0.13 0>}
   B=Constant3Vector() {constant=<rgba 0 0.0833 1 0>}
   Alpha=Saturate(None=ComponentMask(None=@n19) {r=true})
  )
  n23 = TextureSample(UVs=@n19) {texture=/Game/Global/Textures/Default/Masks_and_Gradients/gradient_stasis}
  n24 = TextureCoordinate()
  n25 = MaterialFunctionCall(Alpha=OneMinus(None=ComponentMask(None=@n24) {r=true}), Min=Constant() {r=0.1}, Max=Constant() {r=0.4}) {materialFunction=&SmoothStep}
  n26 = Multiply(
   A=MaterialFunctionCall(Alpha=@n23.R, Min=Constant() {r=0.97}, Max=Constant() {r=1}) {materialFunction=&SmoothStep}.Result
   B=@n25.Result
  )
  n27 = Constant() {r=1}
  n28 = Constant3Vector() {constant=<rgba 0.2591 0.448 0 0>}
  n29 = Multiply(
   A=MaterialFunctionCall(Alpha=@n23.R, Min=Constant() {r=0.4}, Max=Constant() {r=0.5}) {desc=Center Mask, materialFunction=&SmoothStep}.Result
   B=TextureSample(UVs=Panner(Coordinate=@n20) {speedX=0.1, speedY=0.25}) {texture=/Game/Global/Textures/Default/Caustics/T_Caustics_Softer}.RGB
  )
  n30 = Multiply(
   A=LinearInterpolate(A=Constant() {r=1}, B=Constant3Vector() {constant=<rgba 0 0.423 0.1339 0>}, Alpha=@n26)
   B=ShadingPathSwitch(Default=LinearInterpolate(A=@n27, B=@n28, Alpha=@n29), Mobile=Constant() {r=1}) {desc=Migrated from FeatureLevelSwitch}
  )
  n31 = Add(A=Multiply(A=@n22) {constB=0.05}, B=Constant3Vector() {constant=<rgba 0.0802 0.1094 0.0027 0>})
  MaterialFunctionCall(Emissive=Multiply(A=@n22, B=Multiply(A=@n30, B=@n31))) {materialFunction=&MF_FX_EmissiveAdjustments}
  n32 = OneMinus(None=ComponentMask(None=TextureCoordinate()) {g=true})
  n33 = Constant() {r=0.9}
  n34 = Constant() {r=1}
  n35 = Constant()
  n36 = ObjectPositionWS()
  n37 = AppendVector(A=Constant2Vector(), B=ComponentMask(None=@n36) {b=true})
  n38 = Subtract(A=CameraPositionWS(), B=@n36)
  n39 = Abs(None=DotProduct(A=Normalize(VectorInput=@n37), B=Normalize(VectorInput=@n38)))
  n40 = Constant() {r=0.9}
  n41 = Constant() {r=30}
  n42 = Constant()
  Multiply(
   A=Multiply(A=If(A=@n32, B=@n33, "A > B"=@n34, "A == B"=@n35, "A < B"=@n35))
   B=AppendVector(A=Constant2Vector(), B=If(A=@n39, B=@n40, "A > B"=@n41, "A == B"=@n42, "A < B"=@n42))
  )
  n43 = TextureCoordinate()
  n44 = OneMinus(None=MaterialFunctionCall(Alpha=ComponentMask(None=@n43) {r=true}, Min=ScalarParameter() {defaultValue=0.8, parameterName=PinchHeight}, Max=Constant() {r=1}) {materialFunction=&SmoothStep}.Result)
  n45 = Constant()
  n46 = MaterialFunctionCall(Alpha=Multiply(A=@n23.R, B=@n25.Result), Min=Constant() {r=0.4}, Max=Constant() {r=0.5}) {desc=Center Mask, materialFunction=&SmoothStep}
  n47 = Multiply(A=Add(A=@n26, B=ShadingPathSwitch(Default=@n29, Mobile=@n45) {desc=Migrated from FeatureLevelSwitch}), B=@n46.Result)
  n48 = Clamp(None=Multiply(A=OneMinus(None=@n21)) {constB=20})
  n49 = Multiply(A=@n46.Result, B=@n48)
  n50 = Multiply(A=Multiply(A=Clamp(None=@n49), B=Multiply(A=@n48, B=@n48))) {constB=0.9}
  n51 = Multiply(A=@n44, B=Saturate(None=Add(A=@n47, B=@n50)))
  n52 = DepthFade(FadeDistance=ScalarParameter() {defaultValue=50, group=Opacity, parameterName=DepthFade})
  n53 = MaterialFunctionCall(
   Fade Length=ScalarParameter() {defaultValue=35, group=Opacity, parameterName=Camera Fade Length}
   Fade Offset=ScalarParameter() {defaultValue=25, parameterName=Camera Fade Offset}
  ) {materialFunction=&CameraDepthFade}
  Multiply(
   A=Multiply(A=@n44, B=Multiply(A=@n44, B=@n51))
   B=LinearInterpolate(A=@n52, B=Multiply(A=@n52, B=@n53.Result), Alpha=ScalarParameter() {defaultValue=1, group=Opacity, parameterName=Use Camera Fade})
  )
  Saturate(None=Multiply(A=@n18, B=ScalarParameter() {defaultValue=4, parameterName=wiggle}))
  Sine(None=Divide(A=OneMinus(None=@n5)) {constB=4})
 ]
}
