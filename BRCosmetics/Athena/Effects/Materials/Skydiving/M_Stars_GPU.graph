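# Masked, lit sprite material for skydiving star particles, drawing animated stars from flipbook mask and normal textures.
# Particle random or dynamic parameters blend two colors and brightness, multiplied by particle color, with a normal-map channel masking emissive.
# Sub-UV frames advance by time or manual control, using a random-SubUV function by default and inline math on mobile.
# Near-camera fade drives a particle-size-scaled offset; mobile drops this, color blending and brightness variation.
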Material /BRCosmetics/Athena/Effects/Materials/Skydiving/M_Stars_GPU
graph {
 properties {
  bUsedWithNanite=true
  bUsedWithNiagaraSprites=true
  bUsedWithParticleSprites=true
  blendMode=BLEND_Masked
  pixelDepthOffsetMode=PDOM_Legacy
  usedShadingModels=Default Lit
 }
 nodes [
  Constant() {r=1}
  Constant() {r=0.25}
  Constant()
  n1 = VectorParameter() {group=Random_Color, parameterName=ColorA, defaultValue=<rgba 1 0 1 1>}
  n2 = VectorParameter() {group=Random_Color, parameterName=ColorB, sortPriority=1, defaultValue=<rgba 0 1 1 1>}
  n3 = StaticSwitchParameter(True=DynamicParameter() {paramNames [Manual Time Control, Lerp Color, Lerp Brightness, Param4]}.Lerp Color, False=ParticleRandom()) {group=SubUV, parameterName=Use Dynamic Parameter Control, sortPriority=5}
  n4 = ParticleColor()
  n5 = Constant() {r=4}
  n6 = DynamicParameter() {paramNames [Manual Time Control, Lerp Color, Lerp Brightness, Param4]}
  n7 = ScalarParameter() {group=SubUV, parameterName=Manual Time Control_1, sortPriority=4}
  n8 = Constant() {r=4}
  n9 = Frac(None=Add(A=ParticleRandom()) {constB=0.05})
  n10 = Multiply(A=Time())
  n11 = StaticSwitch(
   True=MaterialFunctionCall!(
    UseTexCoord 1
    Rows=@n5
    Columns=@n5
    Particle Random Value=StaticSwitchParameter(True=@n6.Manual Time Control, False=@n7) {parameterName=Use Dynamic Parameter Control}
   ) {materialFunction=/Game/Effects/Fort_Effects/Materials/_Functions/MF_RandomSubUV_GPU_Sprites}.Result
   False=MaterialFunctionCall(Rows=@n8, Columns=@n8, Particle Random Value=Add(A=@n9, B=@n10)) {materialFunction=&MF_RandomSubUV_GPU_Sprites}.Result
   Value=StaticBoolParameter() {group=SubUV, parameterName=Use Manual Control, sortPriority=3}
  )
  n12 = TextureSampleParameter2D(UVs=@n11) {group=Texture, parameterName=Alpha Mask, samplerType=SAMPLERTYPE_Grayscale, texture=/BRCosmetics/Athena/Effects/Textures/T_Stars_Mask}
  n13 = StaticComponentMaskParameter(None=@n12.RGB) {defaultG=true, group=Texture, parameterName=Alpha Component Mask, sortPriority=1}
  n14 = DynamicParameter() {paramNames [Manual Time Control, Lerp Color, Lerp Brightness, Param4]}
  n15 = ScalarParameter() {group=SubUV, parameterName=Manual Time Control, sortPriority=4}
  n16 = Multiply(
   A=ScalarParameter() {defaultValue=2, group=SubUV, parameterName=Timer Speed, sortPriority=2}
   B=StaticSwitch(
    True=StaticSwitchParameter(True=@n14.Manual Time Control, False=@n15) {group=SubUV, parameterName=Use Dynamic Parameter Control, sortPriority=5}
    False=Time() {period=1}
    Value=StaticBoolParameter() {group=SubUV, parameterName=Use Manual Control, sortPriority=3}
   )
  )
  n17 = ScalarParameter() {defaultValue=4, group=SubUV, parameterName=U Images}
  n18 = Floor(None=Multiply(A=Frac(None=@n16), B=@n17))
  n19 = Divide(B=@n17) {constB=1}
  n20 = TextureCoordinate()
  n21 = ComponentMask(None=@n20) {r=true}
  n22 = ComponentMask(None=@n20) {g=true}
  n23 = ScalarParameter() {defaultValue=4, group=SubUV, parameterName=V_Images, sortPriority=1}
  n24 = Divide(B=@n23) {constB=1}
  n25 = Divide(A=@n16, B=@n23)
  n26 = Floor(None=Multiply(A=Frac(None=@n25), B=@n23))
  n27 = AppendVector(A=Add(A=Divide(A=@n18, B=@n17), B=Multiply(A=@n19, B=@n21)), B=Add(A=Multiply(A=@n22, B=@n24), B=Divide(A=@n26, B=@n23)))
  n28 = TextureSampleParameter2D(UVs=@n27) {group=Texture, parameterName=Alpha Mask, samplerType=SAMPLERTYPE_Grayscale, texture=&T_Stars_Mask}
  n29 = StaticComponentMaskParameter(None=@n28.RGB) {defaultG=true, group=Texture, parameterName=Alpha Component Mask, sortPriority=1}
  n30 = Multiply(
   A=ShadingPathSwitch(Default=LinearInterpolate(A=@n1.RGB, B=@n2.RGB, Alpha=@n3), Mobile=@n2.RGB) {desc=Migrated from FeatureLevelSwitch}
   B=ShadingPathSwitch(Default=Multiply(A=@n4.RGB, B=@n13), Mobile=Multiply(A=@n29, B=@n4.RGB)) {desc=Migrated from FeatureLevelSwitch}
  )
  n31 = DynamicParameter() {paramNames [Manual Time Control, Lerp Color, Lerp Brightness, Param4]}
  n32 = ParticleRandom()
  n33 = Multiply(
   A=@n30
   B=LinearInterpolate(
    A=ScalarParameter() {defaultValue=1, group=Random_Color, parameterName=BrightLow, sortPriority=2}
    B=ScalarParameter() {defaultValue=8, group=Random_Color, parameterName=BrightHigh, sortPriority=3}
    Alpha=StaticSwitchParameter(True=@n31.Lerp Brightness, False=@n32) {group=SubUV, parameterName=Use Dynamic Parameter Control, sortPriority=5}
   )
  )
  n34 = TextureSampleParameter2D(UVs=@n11) {group=Texture, parameterName=Normal, samplerType=SAMPLERTYPE_Normal, sortPriority=3, texture=/BRCosmetics/Athena/Effects/Textures/T_Stars_NM}
  n35 = TextureSampleParameter2D(UVs=@n27) {group=Texture, parameterName=Normal, samplerType=SAMPLERTYPE_Normal, sortPriority=3, texture=&T_Stars_NM}
  n36 = ShadingPathSwitch(Default=@n34.RGB, Mobile=@n35.RGB) {desc=Migrated from FeatureLevelSwitch}
  n37 = Multiply(A=@n13, B=DotProduct(A=@n36, B=Constant3Vector() {constant=<rgba 0 0 1 0>}))
  n38 = DotProduct(A=@n36, B=Constant3Vector() {constant=<rgba 1 0 0 0>})
  MaterialFunctionCall(Emissive=Multiply(
    A=ShadingPathSwitch(Default=@n33, Mobile=@n30) {desc=Migrated from FeatureLevelSwitch}
    B=StaticSwitchParameter(True=@n37, False=@n38) {parameterName=Use B Channel Normal As Emisive Mask}
   )) {materialFunction=&MF_FX_EmissiveAdjustments}
  Multiply(A=@n30, B=ScalarParameter() {defaultValue=0.2, group=Random_Color, parameterName=BaseColorStrength, sortPriority=4})
  ShadingPathSwitch(Default=@n13, Mobile=@n29) {desc=Migrated from FeatureLevelSwitch}
  n39 = MaterialFunctionCall(Distance=ScalarParameter() {defaultValue=216.0992, group=Near Camera Fade, parameterName=CamFadeDist}) {materialFunction=&Near_Camera_Fade}
  n40 = Transform(None=AppendVector(A=TextureCoordinate(), B=Constant()))
  n41 = ComponentMask(None=ParticleSize()) {r=true}
  ShadingPathSwitch(Default=Multiply(A=LinearInterpolate(Alpha=@n39.Result) {constA=-1, constB=0}, B=Multiply(A=@n40, B=@n41)), Mobile=Constant()) {desc=Migrated from FeatureLevelSwitch}
 ]
}
